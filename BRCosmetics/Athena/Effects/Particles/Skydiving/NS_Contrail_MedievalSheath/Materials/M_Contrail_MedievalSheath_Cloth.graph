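# Masked, two-sided lit material for cloth mesh particles in the MedievalSheath skydiving contrail Niagara effect.
# Per-particle randomized noise, weighted along UV V and driven by the dynamic Erosion parameter, erodes the masked fabric, with a reflected-gradient glowing burn edge.
# Vertex offset math layers Perlin-driven twist and sine flapping along the length, using RotateAboutAxis.
# The same rotations are applied to the flattened fabric normal to keep shading consistent.
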Material /BRCosmetics/Athena/Effects/Particles/Skydiving/NS_Contrail_MedievalSheath/Materials/M_Contrail_MedievalSheath_Cloth
graph {
 properties {
  bAutomaticallySetUsageInEditor=false
  bUsedWithNiagaraMeshParticles=true
  blendMode=BLEND_Masked
  twoSided=true
  usedShadingModels=Default Lit
 }
 nodes [
  MaterialFunctionCall() {materialFunction=&DebugTimeSine}
  n1 = TextureCoordinate()
  n2 = TextureSample(UVs=@n1) {
   samplerType=SAMPLERTYPE_Masks
   texture=/BRCosmetics/Athena/Effects/Particles/Skydiving/NS_Contrail_MedievalSheath/Textures/T_MedievalSheathContrail_RT_S
  }
  n3 = Add(A=TextureCoordinate(), B=ParticleRandom())
  n4 = AppendVector(
   A=ScalarParameter() {defaultValue=0.3, parameterName=Fire Noise UV Scale U}
   B=ScalarParameter() {defaultValue=1, parameterName=Fire Noise UV Scale V}
  )
  n5 = Multiply(
   A=TextureSample(UVs=Multiply(A=@n3, B=@n4)) {texture=/BRCosmetics/Athena/Effects/Particles/Skydiving/NS_Hightower_Date/T_AtmosphericCloudNoise_Vertical}.R
   B=ComponentMask(None=@n1) {g=true}
  )
  n6 = DynamicParameter() {defaultValue=<rgba 0.5 1 1 1>, paramNames [Crystal Cloth Lerp, Time Offset, Erosion, EmissiveRandom]}
  n7 = MaterialFunctionCall!(
   Gradient=@n5
   Erosion Amount=LinearInterpolate(
    A=ScalarParameter() {parameterName=Fire Erosion Base}
    B=ScalarParameter() {defaultValue=1, parameterName=Fire Erosion Tip}
    Alpha=@n6.Erosion
   )
   Softness=ScalarParameter() {defaultValue=0.05, parameterName=Fire Erosion Burning Edge Size}
  ) {materialFunction=/BRCosmetics/Athena/Effects/Particles/Skydiving/NS_Contrail_DarkAndLight/Functions/MF_Erosion}
  n8 = TextureSample(UVs=@n1) {
   desc=Set to shared:clamp to fix jira FORT-845544
   samplerSource=SSM_Clamp_WorldGroupSettings
   samplerType=SAMPLERTYPE_Masks
   texture=/BRCosmetics/Athena/Effects/Particles/Skydiving/NS_Contrail_MedievalSheath/Textures/T_MedievalSheathContrail_RT_M
  }
  n9 = Multiply(A=OneMinus(None=@n7.Increased Range), B=@n8.R)
  n10 = TextureSample(UVs=@n1) {
   samplerType=SAMPLERTYPE_Normal
   texture=/BRCosmetics/Athena/Effects/Particles/Skydiving/NS_Contrail_MedievalSheath/Textures/T_MedievalSheathContrail_RT_N
  }
  n11 = MaterialFunctionCall(Normal=@n10.RGB, Flatness=ScalarParameter() {defaultValue=0.666, parameterName=Normal Map Flatten}) {materialFunction=&FlattenNormal}
  n12 = BreakMaterialAttributes!(Attr=MakeMaterialAttributes!(
    BaseColor=TextureSample(UVs=@n1) {texture=/BRCosmetics/Athena/Effects/Particles/Skydiving/NS_Contrail_MedievalSheath/Textures/T_MedievalSheathContrail_RT_D}.RGB
    Metallic=@n2.B
    Specular=@n2.B
    Roughness=@n2.B
    Anisotropy
    EmissiveColor
    Opacity=@n9
    OpacityMask=@n9
    Normal=NamedRerouteDeclaration(None=@n11.Result) {name=Fabric Normal, nodeColor=<rgba 1 0 0.9844 1>}
    Tangent
    WorldPositionOffset
    SubsurfaceColor
    ClearCoat
    ClearCoatRoughness
    AmbientOcclusion
    Refraction
    CustomizedUVs_0
    CustomizedUVs_1
    CustomizedUVs_2
    CustomizedUVs_3
    CustomizedUVs_4
    CustomizedUVs_5
    CustomizedUVs_6
    CustomizedUVs_7
    PixelDepthOffset
    ShadingModel
    Displacement
   ))
  MaterialFunctionCall(Emissive=@n12.EmissiveColor) {materialFunction=&MF_FX_EmissiveAdjustments}
  n13 = VectorParameter() {parameterName=Fire Burning Edge Color, defaultValue=<rgba 22 0.3667 0 0.5>}
  Multiply(
   A=Multiply(A=@n6.EmissiveRandom, B=@n13.RGB)
   B=MaterialFunctionCall!(Gradient=@n7.Increased Range, Center Point=@n13.A) {materialFunction=/Game/Effects/Fort_Effects/MaterialFunctions/MF_GradientReflected}.Result
  )
  Multiply(A=@n5, B=@n8.R)
  n14 = NamedRerouteUsage()
  n15 = Multiply(
   A=Add(
    A=Time()
    B=DynamicParameter() {defaultValue=<rgba 0.5 1 1 1>, paramNames [Crystal Cloth Lerp, Time Offset, Erosion, EmissiveRandom]}.Time Offset
   )
   B=ScalarParameter() {defaultValue=2, parameterName=Flapping Speed}
  )
  n16 = Multiply(A=ComponentMask(None=TextureCoordinate()) {g=true}, B=ScalarParameter() {defaultValue=-2, parameterName=Flapping Time Offset Along Length})
  n17 = Multiply(A=Sine(None=Add(A=@n15, B=@n16)), B=ScalarParameter() {defaultValue=0.01, parameterName=Rotation Range}) {constB=0}
  n18 = MaterialFunctionCall() {materialFunction=&ComponentPivotLocation}
  n19 = Normalize(VectorInput=Transform(None=Constant3Vector() {constant=<rgba 0 0 1 0>}) {transformSourceType=TRANSFORMSOURCE_Instance})
  n20 = ComponentMask(None=TextureCoordinate()) {g=true}
  n21 = ScalarParameter() {defaultValue=-0.62, parameterName=Twist Time Offset Along Length}
  n22 = Panner(
   Coordinate=Constant()
   Time=Add(A=Multiply(A=@n20, B=@n21), B=Time())
   Speed=AppendVector(
    A=ScalarParameter() {defaultValue=0.35, parameterName=Twist Panner Speed}
    B=ScalarParameter() {defaultValue=0.55, parameterName=Twist Panner Speed V}
   )
  ) {speedX=1, speedY=2}
  n23 = TextureCoordinate()
  n24 = Multiply(
   A=Subtract(A=TextureSample(UVs=@n22) {samplerType=SAMPLERTYPE_LinearColor, texture=/BRCosmetics/Weapons/FORT_Melee/Pickaxe_RhombCamo/FX/T_Pickaxe_RhombCamo_Perlin}.R) {constB=0.5}
   B=Multiply(A=ComponentMask(None=@n23) {g=true}, B=ScalarParameter() {defaultValue=0.62, parameterName=Twist Range})
  )
  n25 = WorldPosition()
  n26 = RotateAboutAxis(NormalizedRotationAxis=@n19, RotationAngle=@n24, PivotPoint=ParticlePositionWS(), Position=@n25.XYZ)
  n27 = Add(A=@n26, B=@n25.XYZ)
  NamedRerouteDeclaration(None=Add(A=RotateAboutAxis(NormalizedRotationAxis=@n14, RotationAngle=@n17, PivotPoint=@n18.Result, Position=@n27), B=@n26)) {name=WPO, nodeColor=<rgba 0 0.1094 1 1>}
  NamedRerouteDeclaration(None=Transform(None=@n12.Normal)) {name=Normal Map WS, nodeColor=<rgba 1 0 0.4453 1>}
  n28 = ComponentMask(None=Subtract(A=CameraPositionWS(), B=ParticlePositionWS())) {g=true, r=true}
  NamedRerouteDeclaration(None=AppendVector(A=Normalize(VectorInput=@n28), B=Constant())) {name=Rotation Axis, nodeColor=<rgba 0.6875 1 0 1>}
  n29 = Constant()
  n30 = NamedRerouteUsage()
  n31 = Normalize(VectorInput=Add(A=RotateAboutAxis(NormalizedRotationAxis=@n19, RotationAngle=@n24, PivotPoint=@n29, Position=@n30), B=@n30))
  n32 = Add(A=RotateAboutAxis(NormalizedRotationAxis=@n14, RotationAngle=@n17, PivotPoint=Constant(), Position=@n31), B=@n31)
  NamedRerouteDeclaration(None=Transform(None=Normalize(VectorInput=@n32)) {transformSourceType=TRANSFORMSOURCE_World, transformType=TRANSFORM_Tangent}) {name=Normal TS, nodeColor=<rgba 0.9219 0 1 1>}
  NamedRerouteUsage()
  NamedRerouteUsage()
 ]
}
